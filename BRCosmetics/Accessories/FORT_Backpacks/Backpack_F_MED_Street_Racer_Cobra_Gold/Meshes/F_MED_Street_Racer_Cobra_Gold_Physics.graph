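# Physics asset for the gold Street Racer Cobra backpack cosmetic, giving its pair of finger-like wings physics bodies.
# Capsule bodies form three three-bone chains (index, mid, pinky) per side, each rooted on a small wing-hand body that joins the pack body.
# Finger joints carry very tight swing and twist limit angles, while both hand-to-pack joints leave angular motion free.
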PhysicsAsset /BRCosmetics/Accessories/FORT_Backpacks/Backpack_F_MED_Street_Racer_Cobra_Gold/Meshes/F_MED_Street_Racer_Cobra_Gold_Physics
bodies {
 dyn_wingHand_l [{shapeType=Capsule, length=0, radius=0.5, center=<2.9394 0.0948 -1.4329>}]
 dyn_wingHand_r [{shapeType=Capsule, length=0, radius=0.5, center=<2.3472 0.0433 -3.5363>}]
 dyn_wingIndex_01_l [{shapeType=Capsule, length=5, radius=3, center=<5 0 0>, rotation=<rot 90 0 0>}]
 dyn_wingIndex_01_r [{shapeType=Capsule, length=5, radius=3, center=<5 0 0>, rotation=<rot 90 0 0>}]
 dyn_wingIndex_02_l [{shapeType=Capsule, length=5, radius=3, center=<5 0 0>, rotation=<rot 90 0 0>}]
 dyn_wingIndex_02_r [{shapeType=Capsule, length=5, radius=3, center=<5 0 0>, rotation=<rot 90 0 0>}]
 dyn_wingIndex_03_l [{shapeType=Capsule, length=5, radius=3, center=<5 0 0>, rotation=<rot 90 0 0>}]
 dyn_wingIndex_03_r [{shapeType=Capsule, length=5, radius=3, center=<5 0 0>, rotation=<rot 90 0 0>}]
 dyn_wingMid_01_l [{shapeType=Capsule, length=5, radius=3, center=<5 0 0>, rotation=<rot 90 0 0>}]
 dyn_wingMid_01_r [{shapeType=Capsule, length=5, radius=3, center=<5 0 0>, rotation=<rot 90 0 0>}]
 dyn_wingMid_02_l [{shapeType=Capsule, length=5, radius=3, center=<5 0 0>, rotation=<rot 90 0 0>}]
 dyn_wingMid_02_r [{shapeType=Capsule, length=5, radius=3, center=<5 0 0>, rotation=<rot 90 0 0>}]
 dyn_wingMid_03_l [{shapeType=Capsule, length=5, radius=3, center=<5 0 0>, rotation=<rot 90 0 0>}]
 dyn_wingMid_03_r [{shapeType=Capsule, length=5, radius=3, center=<5 0 0>, rotation=<rot 90 0 0>}]
 dyn_wingPinky_01_l [{shapeType=Capsule, length=2, radius=3, center=<3.5 0 0>, rotation=<rot 90 0 0>}]
 dyn_wingPinky_01_r [{shapeType=Capsule, length=2, radius=3, center=<3.5 0 0>, rotation=<rot 90 0 0>}]
 dyn_wingPinky_02_l [{shapeType=Capsule, length=2, radius=3, center=<3.5 0 0>, rotation=<rot 90 0 0>}]
 dyn_wingPinky_02_r [{shapeType=Capsule, length=2, radius=3, center=<3.5 0 0>, rotation=<rot 90 0 0>}]
 dyn_wingPinky_03_l [{shapeType=Capsule, length=2, radius=3, center=<3.5 0 0>, rotation=<rot 90 0 0>}]
 dyn_wingPinky_03_r [{shapeType=Capsule, length=2, radius=3, center=<3.5 0 0>, rotation=<rot 90 0 0>}]
 pack [{shapeType=Capsule, length=0, radius=0.5, center=<-0.7967 -4.3509 9.1609>, rotation=<rot 90 0 0>}]
}
constraints [
 {bone1Name=dyn_wingIndex_01_l, bone2Name=dyn_wingHand_l, swing1LimitDegrees=3, swing2LimitDegrees=3, twistLimitDegrees=1}
 {bone1Name=dyn_wingIndex_02_l, bone2Name=dyn_wingIndex_01_l, swing1LimitDegrees=3, swing2LimitDegrees=3, twistLimitDegrees=1}
 {bone1Name=dyn_wingIndex_03_l, bone2Name=dyn_wingIndex_02_l, swing1LimitDegrees=3, swing2LimitDegrees=3, twistLimitDegrees=1}
 {bone1Name=dyn_wingMid_01_l, bone2Name=dyn_wingHand_l, swing1LimitDegrees=3, swing2LimitDegrees=3, twistLimitDegrees=1}
 {bone1Name=dyn_wingMid_02_l, bone2Name=dyn_wingMid_01_l, swing1LimitDegrees=3, swing2LimitDegrees=3, twistLimitDegrees=1}
 {bone1Name=dyn_wingMid_03_l, bone2Name=dyn_wingMid_02_l, swing1LimitDegrees=3, swing2LimitDegrees=3, twistLimitDegrees=1}
 {bone1Name=dyn_wingPinky_01_l, bone2Name=dyn_wingHand_l, swing1LimitDegrees=3, swing2LimitDegrees=3, twistLimitDegrees=1}
 {bone1Name=dyn_wingPinky_02_l, bone2Name=dyn_wingPinky_01_l, swing1LimitDegrees=3, swing2LimitDegrees=3, twistLimitDegrees=1}
 {bone1Name=dyn_wingPinky_03_l, bone2Name=dyn_wingPinky_02_l, swing1LimitDegrees=3, swing2LimitDegrees=3, twistLimitDegrees=1}
 {bone1Name=dyn_wingIndex_01_r, bone2Name=dyn_wingHand_r, swing1LimitDegrees=3, swing2LimitDegrees=3, twistLimitDegrees=1}
 {bone1Name=dyn_wingIndex_02_r, bone2Name=dyn_wingIndex_01_r, swing1LimitDegrees=3, swing2LimitDegrees=3, twistLimitDegrees=1}
 {bone1Name=dyn_wingIndex_03_r, bone2Name=dyn_wingIndex_02_r, swing1LimitDegrees=3, swing2LimitDegrees=3, twistLimitDegrees=1}
 {bone1Name=dyn_wingMid_01_r, bone2Name=dyn_wingHand_r, swing1LimitDegrees=3, swing2LimitDegrees=3, twistLimitDegrees=1}
 {bone1Name=dyn_wingMid_02_r, bone2Name=dyn_wingMid_01_r, swing1LimitDegrees=3, swing2LimitDegrees=3, twistLimitDegrees=1}
 {bone1Name=dyn_wingMid_03_r, bone2Name=dyn_wingMid_02_r, swing1LimitDegrees=3, swing2LimitDegrees=3, twistLimitDegrees=1}
 {bone1Name=dyn_wingPinky_01_r, bone2Name=dyn_wingHand_r, swing1LimitDegrees=3, swing2LimitDegrees=3, twistLimitDegrees=1}
 {bone1Name=dyn_wingPinky_02_r, bone2Name=dyn_wingPinky_01_r, swing1LimitDegrees=3, swing2LimitDegrees=3, twistLimitDegrees=1}
 {bone1Name=dyn_wingPinky_03_r, bone2Name=dyn_wingPinky_02_r, swing1LimitDegrees=3, swing2LimitDegrees=3, twistLimitDegrees=1}
 {
  bone1Name=dyn_wingHand_l
  bone2Name=pack
  swing1LimitDegrees=45
  swing1Motion=Free
  swing2LimitDegrees=45
  swing2Motion=Free
  twistLimitDegrees=45
  twistMotion=Free
 }
 {
  bone1Name=dyn_wingHand_r
  bone2Name=pack
  swing1LimitDegrees=45
  swing1Motion=Free
  swing2LimitDegrees=45
  swing2Motion=Free
  twistLimitDegrees=45
  twistMotion=Free
 }
]
