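# Animation Blueprint for the Street Racer Cobra Gold backpack cosmetic, posing its wings for skydiving, targeting and running.
# A state machine picks Base, Sprint or Float poses from skydive, parachute and targeting flags.
# Root yaw follows the owner's spine twist, then RigidBody physics adds secondary motion before the pose is cached.
# When not skydiving, additive Sprint and Base poses weighted by pawn speed layer over that cache.
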AnimBlueprint /BRCosmetics/Accessories/FORT_Backpacks/Backpack_F_MED_Street_Racer_Cobra_Gold/Meshes/F_MED_Street_Racer_Cobra_Gold_AnimBP
target_skeleton=/Game/Accessories/FORT_Backpacks/Backpack_F_MED_Street_Racer_Cobra_Gold/Meshes/F_MED_Street_Racer_Cobra_Gold_Skeleton
graphs [
 AnimGraph {nodes [
   n1 = Get PawnSpeedRemap() -> (PawnSpeedRemap:double)
   n2 = Street_Racer_Cobra_Gold_ADD_PoseAsset : Sprint!(PoseWeight:float=@n1) -> (Pose:Pose Link Structure) {
    is=title AnimGraphNode_PoseByName
    poseAsset=&/Street_Racer_Cobra_Gold_ADD_PoseAsset
    poseName=Sprint
   }
   n3 = Apply Additive(Base=Use cached pose 'preSprintCache'(), Additive=@n2) {alphaBoolBlend {blendOption=Linear}}
   n4 = Street_Racer_Cobra_Gold_ADD_PoseAsset : Base!(PoseWeight:float=@n1) -> (Pose:Pose Link Structure) {
    is=title AnimGraphNode_PoseByName
    poseAsset=&Street_Racer_Cobra_Gold_ADD_PoseAsset
    poseName=Base
   }
   Output Pose(Result=Blend Poses by bool(
     BlendPose_0=Use cached pose 'preSprintCache'()
     BlendPose_1=Apply Additive(Base=@n3, Additive=@n4) {alphaBoolBlend {blendOption=Linear}}
     BlendTime_0=0.3
     BlendTime_1=0.3
     bActiveValue=Get bIsSkyDiving()
    ) {blendType=Linear}) {layerGroup=None}
   AnimGraphNode_StateMachine_0 = wingPosition() {is=title AnimGraphNode_StateMachine}
   n5 = "Transform (Modify) Bone - Bone: root"(
    ComponentPose=Local To Component(LocalPose=@AnimGraphNode_StateMachine_0)
    Translation="0,3,0"
    Rotation=Get backpackRotate() -> (backpackRotate:Rotator)
   ) {
    rotationMode=BMM_Additive
    rotationSpace=BCS_BoneSpace
    translationMode=BMM_Additive
    alphaBoolBlend {blendOption=Linear}
    boneToModify {boneName=root}
    translation {y=3}
   }
   n6 = RigidBody(ComponentPose=@n5) {
    bClampLinearTranslationLimitToRefPose=true
    bOverrideWorldGravity=true
    overridePhysicsAsset=&/F_MED_Street_Racer_Cobra_Gold_Physics
    simulationSpace=BaseBoneSpace
    alphaBoolBlend {blendOption=Linear}
    baseBoneRef {boneName=pack}
    componentAppliedLinearAccClamp=<5000 5000 5000>
    componentLinearAccScale=<0.3 0.3 0.3>
    componentLinearVelScale=<0.4 0.4 0.4>
    overrideWorldGravity {z=-980}
   }
   Save cached pose 'preSprintCache'(Pose=Component To Local(ComponentPose=@n6)) {cacheName=preSprintCache}
  ]}
 wingPosition {
  entry=Base
  path=AnimGraph.AnimGraphNode_StateMachine_0.wingPosition
  ids=id
  nodes [
   AnimStateNode_0 = Base() {is=title AnimStateNode}
   AnimStateNode_1 = Diving() {is=title AnimStateNode}
   AnimStateNode_2 = Floating() {is=title AnimStateNode}
   AnimStateNode_3 = Targeting() {is=title AnimStateNode}
   AnimStateTransitionNode_0 = AnimStateTransitionNode()
   AnimStateTransitionNode_1 = AnimStateTransitionNode()
   AnimStateTransitionNode_2 = AnimStateTransitionNode()
   AnimStateTransitionNode_3 = AnimStateTransitionNode()
   AnimStateTransitionNode_5 = AnimStateTransitionNode()
   AnimStateTransitionNode_6 = AnimStateTransitionNode()
   AnimStateTransitionNode_7 = AnimStateTransitionNode()
  ]
  links [
   {
    dst=Diving
    rule=AnimGraph.AnimGraphNode_StateMachine_0.wingPosition.AnimStateTransitionNode_0.Transition
    src=Base
    via=AnimStateTransitionNode_0
   }
   {
    dst=Targeting
    rule=AnimGraph.AnimGraphNode_StateMachine_0.wingPosition.AnimStateTransitionNode_6.Transition
    src=Base
    via=AnimStateTransitionNode_6
   }
   {
    dst=Base
    rule=AnimGraph.AnimGraphNode_StateMachine_0.wingPosition.AnimStateTransitionNode_3.Transition
    src=Diving
    via=AnimStateTransitionNode_3
   }
   {
    dst=Floating
    rule=AnimGraph.AnimGraphNode_StateMachine_0.wingPosition.AnimStateTransitionNode_1.Transition
    src=Diving
    via=AnimStateTransitionNode_1
   }
   {
    dst=Base
    rule=AnimGraph.AnimGraphNode_StateMachine_0.wingPosition.AnimStateTransitionNode_2.Transition
    src=Floating
    via=AnimStateTransitionNode_2
   }
   {
    dst=Diving
    rule=AnimGraph.AnimGraphNode_StateMachine_0.wingPosition.AnimStateTransitionNode_5.Transition
    src=Floating
    via=AnimStateTransitionNode_5
   }
   {
    dst=Base
    rule=AnimGraph.AnimGraphNode_StateMachine_0.wingPosition.AnimStateTransitionNode_7.Transition
    src=Targeting
    via=AnimStateTransitionNode_7
   }
  ]
 }
 Base {
  path=AnimGraph.AnimGraphNode_StateMachine_0.wingPosition.AnimStateNode_0.Base
  nodes [
   n1 = Street_Racer_Cobra_Gold_PoseAsset : Base!(PoseWeight:float=1) -> (Pose:Pose Link Structure) {
    is=title AnimGraphNode_PoseByName
    poseAsset=&/Street_Racer_Cobra_Gold_PoseAsset
    poseName=Base
   }
   Output Animation Pose(Result=@n1)
  ]
 }
 Diving {
  path=AnimGraph.AnimGraphNode_StateMachine_0.wingPosition.AnimStateNode_1.Diving
  nodes [
   n1 = Street_Racer_Cobra_Gold_PoseAsset : Sprint!(PoseWeight:float=1) -> (Pose:Pose Link Structure) {
    is=title AnimGraphNode_PoseByName
    poseAsset=&Street_Racer_Cobra_Gold_PoseAsset
    poseName=Sprint
   }
   Output Animation Pose(Result=@n1)
  ]
 }
 Floating {
  path=AnimGraph.AnimGraphNode_StateMachine_0.wingPosition.AnimStateNode_2.Floating
  nodes [
   n1 = Street_Racer_Cobra_Gold_PoseAsset : Float!(PoseWeight:float=1) -> (Pose:Pose Link Structure) {
    is=title AnimGraphNode_PoseByName
    poseAsset=&Street_Racer_Cobra_Gold_PoseAsset
    poseName=Float
   }
   Output Animation Pose(Result=@n1)
  ]
 }
 Targeting {
  path=AnimGraph.AnimGraphNode_StateMachine_0.wingPosition.AnimStateNode_3.Targeting
  nodes [
   n1 = Street_Racer_Cobra_Gold_PoseAsset : Sprint(PoseWeight=1) {
    poseAsset=&Street_Racer_Cobra_Gold_PoseAsset
    poseName=Sprint
   }
   Output Animation Pose(Result=@n1)
  ]
 }
 Transition {
  path=AnimGraph.AnimGraphNode_StateMachine_0.wingPosition.AnimStateTransitionNode_0.Transition
  nodes [Result(bCanEnterTransition=Get bIsSkydiveDiveMode())]
 }
 Transition {
  path=AnimGraph.AnimGraphNode_StateMachine_0.wingPosition.AnimStateTransitionNode_1.Transition
  nodes [Result(bCanEnterTransition=Get bIsSkydiveFloating())]
 }
 Transition {
  path=AnimGraph.AnimGraphNode_StateMachine_0.wingPosition.AnimStateTransitionNode_2.Transition
  nodes [Result(bCanEnterTransition=Get bIsParachuteOpened())]
 }
 Transition {
  path=AnimGraph.AnimGraphNode_StateMachine_0.wingPosition.AnimStateTransitionNode_3.Transition
  nodes [Result(bCanEnterTransition=Get bIsParachuteOpened())]
 }
 Transition {
  path=AnimGraph.AnimGraphNode_StateMachine_0.wingPosition.AnimStateTransitionNode_5.Transition
  nodes [Result(bCanEnterTransition=Get bIsSkydiveDiveMode())]
 }
 Transition {
  path=AnimGraph.AnimGraphNode_StateMachine_0.wingPosition.AnimStateTransitionNode_6.Transition
  nodes [Result(bCanEnterTransition=Get bIsTargeting())]
 }
 Transition {
  path=AnimGraph.AnimGraphNode_StateMachine_0.wingPosition.AnimStateTransitionNode_7.Transition
  nodes [Result(bCanEnterTransition=NOT Boolean(A=Get bIsTargeting()))]
 }
 EventGraph {nodes [
   Event BlueprintUpdateAnimation()
   n1 = GetOwnerSkeletalMeshComponent()
   n2 = BreakRotator(InRot=BreakTransform(InTransform=GetDeltaTransformFromRefPose(self=@n1, BoneName=spine_05, BaseName=spine_01)).Rotation)
   > Set backpackRotate(backpackRotate:Rotator=MakeRotator(Yaw=MapRangeClamped(Value=@n2.Yaw, InRangeA=-5, InRangeB=-30, OutRangeB=25))) -> (Output_Get:Rotator)
   > Set PawnSpeedRemap(PawnSpeedRemap:double=MapRangeClamped(Value=Get PawnSpeed2D(), InRangeA=0, InRangeB=549)) -> (Output_Get:double)
   TryGetPawnOwner()
  ]}
]
variables [
 backpackRotate {type=Rotator}
 PawnSpeedRemap {type=number}
 wingBase_l {type=Rotator}
 wingBase_r {type=Rotator}
 wingHand_l {type=Rotator}
 wingHand_r {type=Rotator}
]
